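# Physics asset for the MatteDecoy medium male body's Pack mesh, defining the bodies and joints for a dynamic jacket.
# Seven three-capsule chains cover the back, sides and front, each jointed to its own tiny MSTR anchor body.
# Pelvis, lower spine and thigh capsules have no constraints and supply body shapes for the jacket to collide against.
# The three back chains lock rotation at their top joint, unlike the front and side chains.
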PhysicsAsset /BRCosmetics/Characters/Player/Male/Medium/Bodies/M_MED_MatteDecoy/Meshes/M_MED_MatteDecoy_Pack_Physics
bodies {
 dyn_jacket_MSTR [{shapeType=Capsule, length=0.505, radius=0.505, shapeName=dyn_jacket_MSTR_capsule, rotation=<rot 0 -90 0>}]
 dyn_jacket_back_l_01 [{shapeType=Capsule, length=30, radius=6, shapeName=dyn_jacket_back_l_01_capsule, center=<2 -6 0>, rotation=<rot 0 0 -11>}]
 dyn_jacket_back_l_02 [{shapeType=Capsule, length=30, radius=6, shapeName=dyn_jacket_back_l_02_capsule, center=<2 -6.1 0>}]
 dyn_jacket_back_l_03 [{shapeType=Capsule, length=30, radius=6, shapeName=dyn_jacket_back_l_03_capsule, center=<2 -5.6 0>}]
 dyn_jacket_back_l_MSTR [{shapeType=Capsule, length=0.505, radius=0.505, shapeName=dyn_jacket_back_l_MSTR_capsule, rotation=<rot 0 -90 0>}]
 dyn_jacket_back_mid_01 [{shapeType=Capsule, length=30, radius=6, shapeName=dyn_jacket_back_mid_01_capsule, center=<2 -5 0>}]
 dyn_jacket_back_mid_02 [{shapeType=Capsule, length=30.0896, radius=5.9104, shapeName=dyn_jacket_back_mid_02_capsule, center=<2 -5.6 0>}]
 dyn_jacket_back_mid_03 [{shapeType=Capsule, length=30, radius=6, shapeName=dyn_jacket_back_mid_03_capsule, center=<2 -5.1 0>}]
 dyn_jacket_back_mid_MSTR [{shapeType=Capsule, length=0.505, radius=0.505, shapeName=dyn_jacket_back_mid_MSTR_capsule, rotation=<rot 0 -90 0>}]
 dyn_jacket_back_r_01 [{shapeType=Capsule, length=30, radius=6, shapeName=dyn_jacket_back_r_01_capsule, center=<-2 5.7 0>, rotation=<rot 0 0 -11>}]
 dyn_jacket_back_r_02 [{shapeType=Capsule, length=30, radius=6, shapeName=dyn_jacket_back_r_02_capsule, center=<-2 6.3 0>}]
 dyn_jacket_back_r_03 [{shapeType=Capsule, length=30.1664, radius=5.8336, shapeName=dyn_jacket_back_r_03_capsule, center=<-2 5.4 0>}]
 dyn_jacket_back_r_MSTR [{shapeType=Capsule, length=0.505, radius=0.505, shapeName=dyn_jacket_back_r_MSTR_capsule, rotation=<rot 0 -90 0>}]
 dyn_jacket_front_l_01 [{
   shapeType=Capsule
   length=30.2652
   radius=5.7348
   shapeName=dyn_jacket_front_l_01_capsule
   center=<2 -5.6 0>
   rotation=<rot 0 -0.0001 -0.1323>
  }]
 dyn_jacket_front_l_02 [{shapeType=Capsule, length=29.9873, radius=6.0127, shapeName=dyn_jacket_front_l_02_capsule, center=<2 -6 0>, rotation=<rot 0 -0.0001 -0.1323>}]
 dyn_jacket_front_l_03 [{
   shapeType=Capsule
   length=30.0129
   radius=5.9871
   shapeName=dyn_jacket_front_l_03_capsule
   center=<1 -6.2 0>
   rotation=<rot 0 -0.0001 -0.1323>
  }]
 dyn_jacket_front_l_MSTR [{shapeType=Capsule, length=0.505, radius=0.505, shapeName=dyn_jacket_front_l_MSTR_capsule, rotation=<rot 0 -90 0>}]
 dyn_jacket_front_r_01 [{shapeType=Capsule, length=30.123, radius=5.877, shapeName=dyn_jacket_front_r_01_capsule, center=<-2 5.6 0>, rotation=<rot 0 0 -4.1151>}]
 dyn_jacket_front_r_02 [{shapeType=Capsule, length=29.9741, radius=6.0259, shapeName=dyn_jacket_front_r_02_capsule, center=<-2 6 0>, rotation=<rot 0 0 -4.1151>}]
 dyn_jacket_front_r_03 [{shapeType=Capsule, length=29.8786, radius=6.1214, shapeName=dyn_jacket_front_r_03_capsule, center=<-1 6.3 0>, rotation=<rot 0 0 -4.1151>}]
 dyn_jacket_front_r_MSTR [{shapeType=Capsule, length=0.505, radius=0.505, shapeName=dyn_jacket_front_r_MSTR_capsule, rotation=<rot 0 -90 0>}]
 dyn_jacket_side_l_01 [{shapeType=Capsule, length=30, radius=6, shapeName=dyn_jacket_side_l_01_capsule, center=<2 -5.5 0>, rotation=<rot 0 0 11.3003>}]
 dyn_jacket_side_l_02 [{shapeType=Capsule, length=30.1134, radius=5.8866, shapeName=dyn_jacket_side_l_02_capsule, center=<2 -5.6 0>, rotation=<rot 0 0 11.3003>}]
 dyn_jacket_side_l_03 [{shapeType=Capsule, length=30, radius=6, shapeName=dyn_jacket_side_l_03_capsule, center=<2 -5 0>, rotation=<rot 0 0 11.3003>}]
 dyn_jacket_side_l_MSTR [{shapeType=Capsule, length=0.505, radius=0.505, shapeName=dyn_jacket_side_l_MSTR_capsule, rotation=<rot 0 -90 0>}]
 dyn_jacket_side_r_01 [{shapeType=Capsule, length=30.4845, radius=5.5155, shapeName=dyn_jacket_side_r_01_capsule, center=<-2 5 0>, rotation=<rot 0 0 11>}]
 dyn_jacket_side_r_02 [{shapeType=Capsule, length=30, radius=6, shapeName=dyn_jacket_side_r_02_capsule, center=<-2 5.9 0>, rotation=<rot 0 0 11>}]
 dyn_jacket_side_r_03 [{shapeType=Capsule, length=30, radius=6, shapeName=dyn_jacket_side_r_03_capsule, center=<-2 5 0>, rotation=<rot 0 0 11>}]
 dyn_jacket_side_r_MSTR [{shapeType=Capsule, length=0.505, radius=0.505, shapeName=dyn_jacket_side_r_MSTR_capsule, rotation=<rot 0 -90 0>}]
 pelvis [
  {shapeType=Capsule, length=8.7598, radius=10.1904, shapeName=pelvis_capsule, center=<0.9657 -1.0602 0>}
  {shapeType=Capsule, length=8.7598, radius=10.1904, shapeName=pelvis_capsule, center=<-3.7342 -1.0602 0>}
 ]
 spine_01 [{shapeType=Capsule, length=7.5399, radius=9.596, shapeName=spine_01_capsule, center=<-1.6855 -0.792 0>}]
 spine_02 [{shapeType=Capsule, length=6.2022, radius=9.1617, shapeName=spine_02_capsule, center=<-3.1906 0 0>}]
 thigh_l [{shapeType=Capsule, length=34.5608, radius=7.5051, shapeName=thigh_l_capsule, center=<-16.5901 0.7775 0.7427>, rotation=<rot 90 0 0>}]
 thigh_r [{shapeType=Capsule, length=34.5608, radius=7.5051, shapeName=thigh_r_capsule, center=<16.5901 -0.7636 -0.7427>, rotation=<rot -90 0 0>}]
}
constraints [
 {
  bone1Name=dyn_jacket_back_l_01
  bone2Name=dyn_jacket_back_l_MSTR
  swing1LimitDegrees=20
  swing1Motion=Locked
  swing2LimitDegrees=10
  swing2Motion=Locked
  twistLimitDegrees=2
  twistMotion=Locked
 }
 {bone1Name=dyn_jacket_back_l_02, bone2Name=dyn_jacket_back_l_01, swing1LimitDegrees=15, swing2LimitDegrees=10, twistLimitDegrees=2}
 {bone1Name=dyn_jacket_back_l_03, bone2Name=dyn_jacket_back_l_02, swing1LimitDegrees=20, swing2LimitDegrees=10, twistLimitDegrees=2}
 {
  bone1Name=dyn_jacket_back_mid_01
  bone2Name=dyn_jacket_back_mid_MSTR
  swing1LimitDegrees=20
  swing1Motion=Locked
  swing2LimitDegrees=10
  swing2Motion=Locked
  twistLimitDegrees=2
  twistMotion=Locked
 }
 {bone1Name=dyn_jacket_back_mid_02, bone2Name=dyn_jacket_back_mid_01, swing1LimitDegrees=15, swing2LimitDegrees=10, twistLimitDegrees=2}
 {bone1Name=dyn_jacket_back_mid_03, bone2Name=dyn_jacket_back_mid_02, swing1LimitDegrees=20, swing2LimitDegrees=10, twistLimitDegrees=2}
 {
  bone1Name=dyn_jacket_back_r_01
  bone2Name=dyn_jacket_back_r_MSTR
  swing1LimitDegrees=20
  swing1Motion=Locked
  swing2LimitDegrees=10
  swing2Motion=Locked
  twistLimitDegrees=2
  twistMotion=Locked
 }
 {bone1Name=dyn_jacket_back_r_02, bone2Name=dyn_jacket_back_r_01, swing1LimitDegrees=15, swing2LimitDegrees=10, twistLimitDegrees=2}
 {bone1Name=dyn_jacket_back_r_03, bone2Name=dyn_jacket_back_r_02, swing1LimitDegrees=20, swing2LimitDegrees=10, twistLimitDegrees=2}
 {bone1Name=dyn_jacket_front_l_01, bone2Name=dyn_jacket_front_l_MSTR, swing1LimitDegrees=20, swing2LimitDegrees=10, twistLimitDegrees=2}
 {bone1Name=dyn_jacket_front_l_02, bone2Name=dyn_jacket_front_l_01, swing1LimitDegrees=20, swing2LimitDegrees=10, twistLimitDegrees=2}
 {bone1Name=dyn_jacket_front_l_03, bone2Name=dyn_jacket_front_l_02, swing1LimitDegrees=20, swing2LimitDegrees=10, twistLimitDegrees=2}
 {bone1Name=dyn_jacket_front_r_01, bone2Name=dyn_jacket_front_r_MSTR, swing1LimitDegrees=20, swing2LimitDegrees=10, twistLimitDegrees=2}
 {bone1Name=dyn_jacket_front_r_02, bone2Name=dyn_jacket_front_r_01, swing1LimitDegrees=20, swing2LimitDegrees=10, twistLimitDegrees=2}
 {bone1Name=dyn_jacket_front_r_03, bone2Name=dyn_jacket_front_r_02, swing1LimitDegrees=20, swing2LimitDegrees=10, twistLimitDegrees=2}
 {bone1Name=dyn_jacket_side_l_01, bone2Name=dyn_jacket_side_l_MSTR, swing1LimitDegrees=20, swing2LimitDegrees=10, twistLimitDegrees=2}
 {bone1Name=dyn_jacket_side_l_02, bone2Name=dyn_jacket_side_l_01, swing1LimitDegrees=20, swing2LimitDegrees=10, twistLimitDegrees=2}
 {bone1Name=dyn_jacket_side_l_03, bone2Name=dyn_jacket_side_l_02, swing1LimitDegrees=20, swing2LimitDegrees=10, twistLimitDegrees=2}
 {bone1Name=dyn_jacket_side_r_01, bone2Name=dyn_jacket_side_r_MSTR, swing1LimitDegrees=20, swing2LimitDegrees=10, twistLimitDegrees=2}
 {bone1Name=dyn_jacket_side_r_02, bone2Name=dyn_jacket_side_r_01, swing1LimitDegrees=20, swing2LimitDegrees=10, twistLimitDegrees=2}
 {bone1Name=dyn_jacket_side_r_03, bone2Name=dyn_jacket_side_r_02, swing1LimitDegrees=20, swing2LimitDegrees=10, twistLimitDegrees=2}
]
